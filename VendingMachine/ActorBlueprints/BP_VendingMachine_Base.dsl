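; Base vending machine Blueprint that handles interaction audio when players approach the machine.
; On BeginPlay it binds to marker interaction changes and to conversation component state changes, though the latter handler is empty.
; A marker change, gated by a scalable-float toggle, uses the Conversation state to choose the approach cue for playback.
; PlayInteractionAudio plays through the audio component, with a spawn-at-location fallback that reads the invalid component's location.
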(event Collision|EventActorBeginOverlap
  (params (OtherActor :type Actor%20Object%20Reference))
  (|CallParentFunction OtherActor) ; This node is disabled and will not be called.\nDrag off pins to build functionality.
)

(event EventTick
  (params (DeltaSeconds :type Float%20%28single-precision%29))
  (|CallParentFunction DeltaSeconds) ; This node is disabled and will not be called.\nDrag off pins to build functionality.
)

(event EventBeginPlay
  (|CallParentFunction)
  (EventDispatchers|BindEventtoOnDialogMarkerInteractionStateChanged (Variables|Components|GetNonPlayerConversationComponent) (delegate Custom|OnInteractionStateChanged))
  (Default|BindEventtoMarkerInteractionChanged (delegate Custom|MarkerInteractionChanged_Event)))

(event Custom|MarkerInteractionChanged_Event
  (CallFunction|TriggerConversationSounds))

(event Custom|OnInteractionStateChanged
  (params (CurrentInteractionState :type EDialogMarkerInteractionState%20Enum :default "Conversation") (InteractingController :type Fort%20Player%20Controller%20Object%20Reference)))

(event Custom|TriggerConversationSounds
  ;[1 1 1] APPROACH / LEAVE SFX LOGIC
  (if (ScalableFloat|GetValueasBool (Variables|SoundEffects|GetTrigger%20Audio%20on%20Player%20Approach))
    (if (Utilities|Enum|Equal(Enum) (Variables|PlayerInteraction|GetMarker%20Interaction%20State) "Conversation")
      (Variables|SoundEffects|SetConversationInteractionCue (Variables|SoundEffects|GetOnPlayerApproachScreenSound))
      (CallFunction|PlayInteractionAudio (Variables|SoundEffects|GetConversationInteractionCue) (Variables|Default|GetInteractionAudioComponent)))))

(fn PlayInteractionAudio
  (params (Sound :type Sound%20Base%20Object%20Reference) (TargetAudioComponent :type Audio%20Component%20Object%20Reference))
  (Utilities|IsValid TargetAudioComponent
    (:Is%20Valid
      (Audio|Components|Audio|SetSound TargetAudioComponent Sound)
      (Audio|Components|Audio|Play TargetAudioComponent 0.0))
    (:Is%20Not%20Valid
      (Audio|SpawnSoundatLocation Sound (Transformation|GetWorldLocation TargetAudioComponent)))))

(fn ConstructionScript
  (|CallParentFunction))
