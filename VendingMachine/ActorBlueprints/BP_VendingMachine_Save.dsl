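; BP_VendingMachine_Save is a Blueprint, apparently save data for the vending machine, that tracks a gold ping count.
; Its IncreaseGoldPingCount function increments the Gold Ping Count variable by one using IncrementInt.
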(fn IncreaseGoldPingCount
  (Math|Integer|IncrementInt (Variables|Default|GetGold%20Ping%20Count)))
